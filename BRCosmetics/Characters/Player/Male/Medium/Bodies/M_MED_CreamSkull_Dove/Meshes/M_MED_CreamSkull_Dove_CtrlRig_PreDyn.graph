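# Pre-dynamics Control Rig for the CreamSkull Dove male medium body, adding procedural corrections to body, jacket and collar bones.
# A shared DeformRig function runs first, then spherical pose readers on spine and thighs drive jacket bases and extra glutes.
# Collar bones get partial weighted parent constraints, then offsets from head-vector dot products.
# Clavicle, right-arm (crouch-switched) and head-to-cape-attach distance drivers add further collar corrections.
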ControlRigBlueprint /BRCosmetics/Characters/Player/Male/Medium/Bodies/M_MED_CreamSkull_Dove/Meshes/M_MED_CreamSkull_Dove_CtrlRig_PreDyn
graphs [
 CollapseNode_ContainedGraph {nodes [
   RigVMFunctionEntryNode()
   Multiply = RigVMAggregateNode(
    C=&RigUnit_GetTransform(Item=(Type=Bone,Name="neck_01"), Space=LocalSpace, bInitial=False).Transform
    A=&RigUnit_GetTransform(Item=(Type=Bone,Name="head_end"), Space=LocalSpace, bInitial=False).Transform
    D=&RigUnit_GetTransform(Item=(Type=Bone,Name="neck_02"), Space=LocalSpace, bInitial=False).Transform
    B=&RigUnit_GetTransform(Item=(Type=Bone,Name="head"), Space=LocalSpace, bInitial=False).Transform
   )
   Unit = /Script/RigVM.RigVMFunction_MathVectorUnit(Value=@Multiply.Result.Translation) {is=struct RigVMUnitNode, template="Unit::Execute(in Value,out Result)"}
   RigVMFunctionReturnNode(Result=@Unit.Result)
  ]}
 Multiply_1_ContainedGraph {nodes [
   n1 = RigVMFunctionEntryNode()
   Multiply = &RigVMFunction_MathTransformMul(A=@n1.A, B=@n1.B) {template="Multiply::Execute(in A,in B,out Result)"}
   RigVMFunctionReturnNode(Result=&RigVMFunction_MathTransformMul(
     B=@n1.D
     A=&RigVMFunction_MathTransformMul(B=@n1.C, A=@Multiply.Result) {template="Multiply::Execute(in A,in B,out Result)"}.Result
    ) {template="Multiply::Execute(in A,in B,out Result)"}.Result)
  ]}
 New Function_ContainedGraph {nodes [
   n1 = RigVMFunctionEntryNode()
   Dot__ = /Script/RigVM.RigVMFunction_MathVectorDot(
    A=&RigVMFunction_MathVectorUnit(Value=RigVMVariableNode() {variable=head_vector}.Value) {template="Unit::Execute(in Value,out Result)"}.Result
    B=&RigVMFunction_MathVectorUnit(Value=&RigUnit_GetTransform(Item=(Type=Bone,Name="collar_09"), Space=LocalSpace, bInitial=False, Item.Name=@n1.BoneIn.Name).Transform.Translation) {template="Unit::Execute(in Value,out Result)"}.Result
   ) {is=struct RigVMUnitNode, template="Dot,|::Execute(in A,in B,out Result)"}
   n2 = &RigVMFunction_MathDoubleRemap(TargetMaximum=1, TargetMinimum=0, Value=@Dot__.Result, SourceMinimum=@n1.s_min, SourceMaximum=@n1.s_max)
   > &RigUnit_OffsetTransformForItem(
    Item=(Type=Bone,Name="None")
    OffsetTransform=(Rotation=(X=0,Y=0,Z=-0.7071,W=0.7071),Translation=(X=-2,Y=0,Z=0),Scale3D=(X=1,Y=1.5,Z=0.1))
    bPropagateToChildren=True
    Weight=@n2.Result
    Item.Name=@n1.BoneIn.Name
   )
   > RigVMFunctionReturnNode(DotAlpha=@n2.Result, BoneOut=@n1.BoneIn.Name)
   Multiply = &RigVMFunction_MathDoubleMul(A=@n1.angle, B=@n2.Result)
   &RigUnit_ModifyTransforms(
    ExecutePin=()
    ItemToModify=((Item=(Type=Bone,Name="None"),Transform=(Rotation=(X=0,Y=0,Z=0,W=1),Translation=(X=0,Y=0,Z=0),Scale3D=(X=1,Y=1,Z=1))))
    Weight=1
    WeightMinimum=0
    ItemToModify.0.Transform.Rotation=&RigVMFunction_MathQuaternionFromEuler(Euler=(X=0,Y=0,Z=0), Euler.Z=@Multiply.Result).Result
    ItemToModify.0.Item.Name=@n1.BoneIn.Name
   )
  ]}
 RigVMFunctionLibrary {nodes [DotRot_CollarHead_1 = RigVMCollapseNode(BoneIn=(Type=None,Name="None"), ExecuteContext=())]}
 RigVMFunction_Sequence_2_ContainedGraph {nodes [
   n1 = RigVMFunctionEntryNode()
   n2 = &RigVMFunction_Sequence(ExecuteContext=@n1.ExecuteContext)
   n3 = &RigVMFunction_Sequence(ExecuteContext=@n2.B)
   RigVMFunctionReturnNode(A=@n2.A, B=@n3.A, C=@n3.B)
  ]}
 RigVMFunction_Sequence_3_ContainedGraph {nodes [
   n1 = RigVMFunctionEntryNode()
   n2 = &RigVMFunction_Sequence(ExecuteContext=@n1.ExecuteContext)
   n3 = &RigVMFunction_Sequence(ExecuteContext=@n2.B)
   RigVMFunctionReturnNode(B=@n3.A, C=@n3.B, A=@n2.A)
  ]}
 RigVMFunction_Sequence_3_ContainedGraph {nodes [
   n1 = RigVMFunctionEntryNode()
   RigVMFunction_Sequence = &RigVMFunction_Sequence(ExecuteContext=@n1.ExecuteContext)
   n2 = &RigVMFunction_Sequence(ExecuteContext=@RigVMFunction_Sequence.B)
   RigVMFunctionReturnNode(A=@RigVMFunction_Sequence.A, B=@n2.A, C=@n2.B)
  ]}
 RigVMFunction_Sequence_4_ContainedGraph {nodes [
   n1 = RigVMFunctionEntryNode()
   n2 = &RigVMFunction_Sequence(ExecuteContext=@n1.ExecuteContext)
   n3 = &RigVMFunction_Sequence(ExecuteContext=@n2.B)
   RigVMFunctionReturnNode(B=@n3.A, C=@n3.B, A=@n2.A)
  ]}
 RigVMFunction_Sequence_5_ContainedGraph {nodes [
   n1 = RigVMFunctionEntryNode()
   n2 = &RigVMFunction_Sequence(ExecuteContext=@n1.ExecuteContext)
   n3 = &RigVMFunction_Sequence(ExecuteContext=@n2.B)
   n4 = &RigVMFunction_Sequence(ExecuteContext=@n3.B)
   n5 = &RigVMFunction_Sequence(ExecuteContext=@n4.B)
   n6 = &RigVMFunction_Sequence(ExecuteContext=@n5.B)
   RigVMFunctionReturnNode(B=@n3.A, C=@n4.A, D=@n5.A, A=@n2.A, E=@n6.A, F=@n6.B)
  ]}
 RigVMFunction_Sequence_6_ContainedGraph {nodes [
   n1 = RigVMFunctionEntryNode()
   n2 = &RigVMFunction_Sequence(ExecuteContext=@n1.ExecuteContext)
   n3 = &RigVMFunction_Sequence(ExecuteContext=@n2.B)
   n4 = &RigVMFunction_Sequence(ExecuteContext=@n3.B)
   RigVMFunctionReturnNode(A=@n2.A, B=@n3.A, C=@n4.A, D=@n4.B)
  ]}
 RigVMModel {nodes [
   RigUnit_BeginExecution = &RigUnit_BeginExecution()
   RigVMFunction_Sequence_1 = RigVMAggregateNode(ExecuteContext=@RigUnit_BeginExecution.ExecutePin)
   FN_DeformRig_FNC = RigVMFunctionReferenceNode(
    DeformRig_ Curve="(deformRig_l_upperarm_up_alpha_72_#1=False,deformRig_r_upperarm_up_alpha_73_#2=False,deformRig_l_upperarm_fr_alpha_74_#3=False,deformRig_r_upperarm_fr_alpha_75_CB51B8 ...(truncated)"
    DeformRig_Bool="(upperarm_1_#4=True,elbow_3_#5=True,wrist_7_#6=True,groin_8_#7=True,glute_17_#8=True,knee_9_40C408324CA261 ...(truncated)"
    DeformRig_Debug=(upperarm_debug_18_#4=False,wrist_debug_20_#6=False,lat_debug_26_#9=False,groinGlute_debug_27_#7=False)
    DeformRig_Value="(upperarm_up_18_#10=(Rotation=(X=0,Y=0,Z=0,W=1),Translation=(X=0,Y=0,Z=8),Scale3D=(X=0.7,Y=1,Z=1)),upperarm_fwd_fr_19_#11=(Rotation=(X=0,Y=0,Z=0,W=1),Translation=(X=0,Y=10,Z=0),Scale3D=(X=0.5,Y=1,Z= ...(truncated)"
    ExecuteContext=@RigVMFunction_Sequence_1.A
   )
   RigVMFunction_Sequence = RigVMAggregateNode(ExecuteContext=@RigVMFunction_Sequence_1.B)
   n1 = &RigUnit_SphericalPoseReader(
    ActiveRegionScaleFactors=(PositiveWidth=1,NegativeWidth=0.2,PositiveHeight=1,NegativeHeight=1)
    ActiveRegionSize=0.1
    Debug=(bDrawDebug=True,bDraw2D=false,DebugScale=25,DebugSegments=20,DebugThickness=0.25)
    DriverAxis=(X=0,Y=1,Z=0)
    DriverItem=(Type=Bone,Name="spine_03")
    FalloffRegionScaleFactors=(PositiveWidth=1,NegativeWidth=1,PositiveHeight=1,NegativeHeight=1)
    FalloffSize=0.2
    FlipHeightScaling=false
    FlipWidthScaling=true
    OptionalParentItem=(Type=Bone,Name="pelvis")
    RotationOffset=(X=90,Y=0,Z=64)
    ExecutePin=@RigVMFunction_Sequence.A
   )
   > &RigUnit_ModifyTransforms(
    ItemToModify=((Item=(Type=Bone,Name="dyn_jacket_fr_r_base"),Transform=(Rotation=(X=-0.2241,Y=-0.483,Z=-0.1294,W=0.8365),Translation=(X=-2,Y=0,Z=0),Scale3D=(X=1,Y=1,Z=1))))
    WeightMinimum=0
    Weight=@n1.OutputParam
   )
   n2 = &RigUnit_SphericalPoseReader(
    ActiveRegionScaleFactors=(PositiveWidth=3.5,NegativeWidth=0.5,PositiveHeight=3,NegativeHeight=3)
    ActiveRegionSize=0.1
    Debug=(bDrawDebug=false,bDraw2D=false,DebugScale=25,DebugSegments=20,DebugThickness=0.25)
    DriverAxis=(X=1,Y=0,Z=0)
    DriverItem=(Type=Bone,Name="thigh_r")
    FalloffRegionScaleFactors=(PositiveWidth=1,NegativeWidth=1,PositiveHeight=1,NegativeHeight=1)
    FalloffSize=0.45
    FlipHeightScaling=false
    FlipWidthScaling=true
    OptionalParentItem=(Type=Bone)
    RotationOffset=(X=-90,Y=0,Z=0)
    ExecutePin=@RigVMFunction_Sequence.B
   )
   AlphaInterp = &RigVMFunction_AlphaInterp(InterpSpeedDecreasing=10, InterpSpeedIncreasing=10, Scale=1, bClampResult=False, Value=@n2.OutputParam)
   > &RigUnit_ModifyTransforms(
    ItemToModify=((Item=(Type=Bone,Name="dyn_jacket_fr_r_base"),Transform=(Rotation=(X=0.1157,Y=0.6022,Z=0.0888,W=-0.7849),Translation=(X=0,Y=-0.9997,Z=0),Scale3D=(X=1,Y=1,Z=1))))
    WeightMinimum=0
    Weight=@AlphaInterp.Result
   )
   > &RigUnit_ModifyTransforms(
    ItemToModify=((Item=(Type=Bone,Name="extra_glute_l"),Transform=(Rotation=(X=0,Y=0,Z=0,W=1),Translation=(X=-2,Y=-10,Z=5),Scale3D=(X=1,Y=1,Z=1))))
    WeightMinimum=0
    Weight=@AlphaInterp.Result
   )
   n3 = &RigUnit_SphericalPoseReader(
    ActiveRegionScaleFactors=(PositiveWidth=3.5,NegativeWidth=0.5,PositiveHeight=3,NegativeHeight=3)
    ActiveRegionSize=0.1
    Debug=(bDrawDebug=false,DebugScale=25,DebugSegments=20,DebugThickness=0.25)
    DriverAxis=(X=-1,Y=0,Z=0)
    DriverItem=(Type=Bone,Name="thigh_l")
    FalloffRegionScaleFactors=(PositiveWidth=1,NegativeWidth=1,PositiveHeight=1,NegativeHeight=1)
    FalloffSize=0.45
    OptionalParentItem=(Type=Bone)
    RotationOffset=(X=90,Y=0,Z=0)
    ExecutePin=@RigVMFunction_Sequence.C
   )
   n4 = &RigVMFunction_AlphaInterp(InterpSpeedDecreasing=2, InterpSpeedIncreasing=2, Scale=1, bClampResult=False, Value=@n3.OutputParam)
   > &RigUnit_ParentConstraint(
    AdvancedSettings=(InterpolationType=Average,RotationOrderForFilter=XZY)
    Child=(Type=Bone,Name="dyn_jacket_fr_l_base")
    Filter=(TranslationFilter=(bX=False,bY=False,bZ=False),RotationFilter=(bX=True,bY=True,bZ=True),ScaleFilter=(bX=True,bY=True,bZ=True))
    Parents=((Item=(Type=Bone,Name="thigh_l"),Weight=1))
    bMaintainOffset=true
    Weight=@n4.Result
   )
   > &RigUnit_OffsetTransformForItem(
    Item=(Type=Bone,Name="dyn_jacket_fr_l_base")
    OffsetTransform=(Rotation=(X=-0.48,Y=0,Z=0,W=-0.8773),Translation=(X=0,Y=0,Z=0),Scale3D=(X=1,Y=1,Z=1))
    bPropagateToChildren=True
    Weight=@n4.Result
   )
   > &RigUnit_ModifyTransforms(
    ItemToModify=((Item=(Type=Bone,Name="extra_glute_r"),Transform=(Rotation=(X=0,Y=0,Z=0,W=1),Translation=(X=-2,Y=10,Z=-5),Scale3D=(X=1,Y=1,Z=1))))
    WeightMinimum=0
    Weight=@n4.Result
   )
   RigVMFunction_Sequence_4 = RigVMAggregateNode(ExecuteContext=@RigVMFunction_Sequence_1.C)
   RigVMFunction_Sequence_5 = RigVMAggregateNode(ExecuteContext=@RigVMFunction_Sequence_4.C)
   n5 = &RigUnit_SphericalPoseReader(
    ActiveRegionScaleFactors=(PositiveWidth=1,NegativeWidth=0.1,PositiveHeight=1,NegativeHeight=0.5)
    ActiveRegionSize=0.3
    Debug=(bDrawDebug=true,bDraw2D=false,bDrawLocalAxes=false,DebugScale=30,DebugSegments=30,DebugThickness=0.2)
    DriverAxis=(X=1,Y=0,Z=0)
    DriverItem=(Type=Bone,Name="clavicle_l")
    FalloffRegionScaleFactors=(PositiveWidth=1,NegativeWidth=0.5,PositiveHeight=1,NegativeHeight=0.5)
    FalloffSize=0.5
    OptionalParentItem=(Type=Bone,Name="None")
    RotationOffset=(X=0,Y=40,Z=180)
    ExecutePin=@RigVMFunction_Sequence_5.C
   )
   n6 = &RigVMFunction_AlphaInterp(InterpSpeedDecreasing=10, InterpSpeedIncreasing=10, Scale=1, bClampResult=False, Value=@n5.OutputParam)
   > &RigUnit_ModifyTransforms(
    ItemToModify="((Item=(Type=Bone,Name=\"extra_collar_06_l\"),Transform=(Rotation=(X=0,Y=0,Z=0,W=1),Translation=(X=0,Y=-0.5,Z=0),Scale3D=(X=1,Y=1,Z=1))),(Item=(Type=Bone,Name=\"extra_collar_05_l\"),Transform=(Rotation=(X=0,Y=0,Z=0,W=1),Translation=(X=0,Y=-1,Z=0),Scale3D=(X=1, ...(truncated)"
    WeightMinimum=0
    Weight=@n6.Result
   )
   > &RigUnit_ModifyTransforms(
    ItemToModify=((Item=(Type=Bone,Name="extra_collar_03_l"),Transform=(Rotation=(X=0,Y=0,Z=0.1736,W=0.9848),Translation=(X=3,Y=0,Z=0),Scale3D=(X=1,Y=1,Z=1))))
    WeightMinimum=0
    Weight=@n6.Result
   )
   > &RigUnit_ModifyTransforms(
    ItemToModify=((Item=(Type=Bone,Name="extra_collar_04_l"),Transform=(Rotation=(X=0,Y=0,Z=0.1736,W=0.9848),Translation=(X=1,Y=1,Z=0),Scale3D=(X=1,Y=1,Z=1))))
    WeightMinimum=0
    Weight=@n6.Result
   )
   n7 = &RigUnit_SphericalPoseReader(
    ActiveRegionScaleFactors=(PositiveWidth=0.3,NegativeWidth=0.3,PositiveHeight=0.3,NegativeHeight=0.3)
    ActiveRegionSize=0.3
    Debug=(bDrawDebug=true,bDraw2D=false,bDrawLocalAxes=false,DebugScale=30,DebugSegments=30,DebugThickness=0.2)
    DriverAxis=(X=-1,Y=0,Z=0)
    DriverItem=(Type=Bone,Name="upperarm_r")
    FalloffRegionScaleFactors=(PositiveWidth=0.1,NegativeWidth=1,PositiveHeight=1,NegativeHeight=0.5)
    FalloffSize=0.5
    OptionalParentItem=(Type=Bone,Name="None")
    RotationOffset=(X=-40,Y=220,Z=0)
    ExecutePin=@RigVMFunction_Sequence_5.E
   )
   RigVMFunction_ControlFlowBranch = &RigVMFunction_ControlFlowBranch(ExecuteContext=@n7.ExecutePin, Condition=RigVMVariableNode() {variable=crouch}.Value)
   n8 = &RigVMFunction_AlphaInterp(InterpSpeedDecreasing=10, InterpSpeedIncreasing=10, Scale=1, bClampResult=False, Value=@n7.OutputParam)
   &RigUnit_ModifyTransforms(
    ItemToModify="((Item=(Type=Bone,Name=\"extra_collar_01_r\"),Transform=(Rotation=(X=0,Y=0,Z=0,W=1),Translation=(X=0,Y=0,Z=0),Scale3D=(X=1,Y=1,Z=1))),(Item=(Type=Bone,Name=\"extra_collar_02_r\"),Transform=(Rotation=(X=0,Y=0,Z=-0.1736,W=0.9848),Translation=(X=0,Y=0,Z=0),Scale3 ...(truncated)"
    WeightMinimum=0
    Weight=@n8.Result
    ExecutePin=@RigVMFunction_ControlFlowBranch.False
   )
   &RigUnit_ModifyTransforms(
    ItemToModify="((Item=(Type=Bone,Name=\"extra_collar_01_r\"),Transform=(Rotation=(X=0,Y=0,Z=0,W=1),Translation=(X=5,Y=0,Z=0),Scale3D=(X=1,Y=1,Z=1))),(Item=(Type=Bone,Name=\"extra_collar_02_r\"),Transform=(Rotation=(X=0,Y=0,Z=0,W=1),Translation=(X=2,Y=-1,Z=4),Scale3D=(X=1,Y=1 ...(truncated)"
    WeightMinimum=0
    Weight=@n8.Result
    ExecutePin=@RigVMFunction_ControlFlowBranch.True
   )
   RigUnit_GetTransform = &RigUnit_GetTransform(Item=(Type=Bone,Name="head_end"), bInitial=False)
   RigVMFunction_MathVectorDistance = &RigVMFunction_MathVectorDistance(
    A=@RigUnit_GetTransform.Transform.Translation
    B=&RigUnit_GetTransform(Item=(Type=Bone,Name="attach_cape"), bInitial=False).Transform.Translation
   )
   Remap = &RigVMFunction_MathDoubleRemap(SourceMaximum=13, SourceMinimum=18, TargetMaximum=1, TargetMinimum=0, Value=@RigVMFunction_MathVectorDistance.Result)
   &RigUnit_ModifyTransforms(
    ItemToModify="((Item=(Type=Bone,Name=\"extra_collar_m\"),Transform=(Rotation=(X=0,Y=0,Z=0,W=1),Translation=(X=2,Y=-3,Z=0),Scale3D=(X=1,Y=2,Z=1))),(Item=(Type=Bone,Name=\"extra_collar_01_l\"),Transform=(Rotation=(X=0,Y=0,Z=0,W=1),Translation=(X=0,Y=-5,Z=0),Scale3D=(X=1,Y=2,Z ...(truncated)"
    WeightMinimum=0
    Weight=@Remap.Result
    ExecutePin=@RigVMFunction_Sequence_5.F
   )
   n9 = &RigUnit_SphericalPoseReader(
    ActiveRegionScaleFactors=(PositiveWidth=1,NegativeWidth=0.1,PositiveHeight=0.5,NegativeHeight=1)
    ActiveRegionSize=0.3
    Debug=(bDrawDebug=true,bDraw2D=false,bDrawLocalAxes=false,DebugScale=30,DebugSegments=30,DebugThickness=0.2)
    DriverAxis=(X=-1,Y=0,Z=0)
    DriverItem=(Type=Bone,Name="clavicle_r")
    FalloffRegionScaleFactors=(PositiveWidth=1,NegativeWidth=0.5,PositiveHeight=1,NegativeHeight=1)
    FalloffSize=0.5
    OptionalParentItem=(Type=Bone,Name="None")
    RotationOffset=(X=0,Y=220,Z=180)
    ExecutePin=@RigVMFunction_Sequence_5.D
   )
   n10 = &RigVMFunction_AlphaInterp(InterpSpeedDecreasing=10, InterpSpeedIncreasing=10, Scale=1, bClampResult=False, Value=@n9.OutputParam)
   > &RigUnit_ModifyTransforms(
    ItemToModify="((Item=(Type=Bone,Name=\"extra_collar_01_r\"),Transform=(Rotation=(X=0,Y=0,Z=0,W=1),Translation=(X=0,Y=0,Z=0),Scale3D=(X=1,Y=1,Z=1))),(Item=(Type=Bone,Name=\"extra_collar_02_r\"),Transform=(Rotation=(X=0,Y=0,Z=0.342,W=0.9397),Translation=(X=0,Y=0,Z=0),Scale3D= ...(truncated)"
    WeightMinimum=0
    Weight=@n10.Result
   )
   > &RigUnit_ModifyTransforms(
    ItemToModify=((Item=(Type=Bone,Name="extra_collar_03_r"),Transform=(Rotation=(X=0,Y=0,Z=0.1736,W=0.9848),Translation=(X=3,Y=0,Z=0),Scale3D=(X=1,Y=1,Z=1))))
    WeightMinimum=0
    Weight=@n10.Result
   )
   > &RigUnit_ModifyTransforms(
    ItemToModify=((Item=(Type=Bone,Name="extra_collar_04_r"),Transform=(Rotation=(X=0,Y=0,Z=0.1736,W=0.9848),Translation=(X=1,Y=1,Z=0),Scale3D=(X=1,Y=1,Z=1))))
    WeightMinimum=0
    Weight=@n10.Result
   )
   RigVMVariableNode(
    Value=&RigVMFunction_MathBoolToFloat(Value=RigVMVariableNode() {variable=target}.Value).Result
    ExecuteContext=@RigVMFunction_Sequence_5.A
   ) {variable=targetAlpha}
   RigVMVariableNode(Value=0.4, ExecuteContext=@RigVMFunction_Sequence_4.A) {variable=contrs}
   > RigVMFunction_Sequence_6 = RigVMAggregateNode()
   n11 = RigVMVariableNode() {variable=contrs}
   &RigUnit_ParentConstraint(
    AdvancedSettings=(RotationOrderForFilter=XZY)
    Child=(Type=Bone,Name="extra_collar_01_l")
    Parents="((Item=(Type=Bone,Name=\"upperarm_twist_01_l\"),Weight=0.0015),(Item=(Type=Bone,Name=\"spine_03\"),Weight=0.0019),(Item=(Type=Bone,Name=\"deform_upperarm_up_l\"),Weight=0.002),(Item=(Type=Bone,Name=\"head\"),Weight=0.0034),(Item=(Type=Bone,Name=\"neck_01\"),Weight=0 ...(truncated)"
    ExecutePin=@RigVMFunction_Sequence_6.A
    Weight=@n11.Value
   )
   > &RigUnit_ParentConstraint(
    AdvancedSettings=(RotationOrderForFilter=XZY)
    Child=(Type=Bone,Name="extra_collar_02_l")
    Parents=((Item=(Type=Bone,Name="head"),Weight=0.0005),(Item=(Type=Bone,Name="neck_01"),Weight=0.0184),(Item=(Type=Bone,Name="neck_02"),Weight=0.0434),(Item=(Type=Bone,Name="clavicle_l"),Weight=0.3416),(Item=(Type=Bone,Name="spine_05"),Weight=0.596))
    Weight=@n11.Value
   )
   > &RigUnit_ParentConstraint(
    AdvancedSettings=(RotationOrderForFilter=XZY)
    Child=(Type=Bone,Name="extra_collar_03_l")
    Parents="((Item=(Type=Bone,Name=\"spine_03\"),Weight=0.0002),(Item=(Type=Bone,Name=\"deform_pec_l\"),Weight=0.0003),(Item=(Type=Bone,Name=\"head\"),Weight=0.0036),(Item=(Type=Bone,Name=\"spine_04\"),Weight=0.0057),(Item=(Type=Bone,Name=\"neck_01\"),Weight=0.0326),(Item=(Type ...(truncated)"
    Weight=@n11.Value
   )
   > &RigUnit_ParentConstraint(
    AdvancedSettings=(RotationOrderForFilter=XZY)
    Child=(Type=Bone,Name="extra_collar_04_l")
    Parents="((Item=(Type=Bone,Name=\"spine_03\"),Weight=0.0212),(Item=(Type=Bone,Name=\"deform_pec_l\"),Weight=0.0274),(Item=(Type=Bone,Name=\"neck_01\"),Weight=0.0391),(Item=(Type=Bone,Name=\"head\"),Weight=0.0507),(Item=(Type=Bone,Name=\"spine_04\"),Weight=0.068),(Item=(Type= ...(truncated)"
    Weight=@n11.Value
   )
   &RigUnit_ParentConstraint(
    AdvancedSettings=(RotationOrderForFilter=XZY)
    Child=(Type=Bone,Name="extra_collar_m")
    Parents=((Item=(Type=Bone,Name="head"),Weight=0.0022),(Item=(Type=Bone,Name="clavicle_r"),Weight=0.0152),(Item=(Type=Bone,Name="neck_01"),Weight=0.0581),(Item=(Type=Bone,Name="neck_02"),Weight=0.1269),(Item=(Type=Bone,Name="spine_05"),Weight=0.7976))
    ExecutePin=@RigVMFunction_Sequence_6.B
    Weight=@n11.Value
   )
   &RigUnit_ParentConstraint(
    AdvancedSettings=(RotationOrderForFilter=XZY)
    Child=(Type=Bone,Name="extra_collar_01_r")
    Parents="((Item=(Type=Bone,Name=\"spine_03\"),Weight=0.0004),(Item=(Type=Bone,Name=\"head\"),Weight=0.0012),(Item=(Type=Bone,Name=\"neck_01\"),Weight=0.0425),(Item=(Type=Bone,Name=\"neck_02\"),Weight=0.0959),(Item=(Type=Bone,Name=\"clavicle_r\"),Weight=0.1607),(Item=(Type=Bo ...(truncated)"
    ExecutePin=@RigVMFunction_Sequence_6.C
    Weight=@n11.Value
   )
   > &RigUnit_ParentConstraint(
    AdvancedSettings=(RotationOrderForFilter=XZY)
    Child=(Type=Bone,Name="extra_collar_02_r")
    Parents=((Item=(Type=Bone,Name="neck_01"),Weight=0.0208),(Item=(Type=Bone,Name="neck_02"),Weight=0.0494),(Item=(Type=Bone,Name="clavicle_r"),Weight=0.3177),(Item=(Type=Bone,Name="spine_05"),Weight=0.612))
    Weight=@n11.Value
   )
   > &RigUnit_ParentConstraint(
    AdvancedSettings=(RotationOrderForFilter=XZY)
    Child=(Type=Bone,Name="extra_collar_03_r")
    Parents="((Item=(Type=Bone,Name=\"head\"),Weight=0.0032),(Item=(Type=Bone,Name=\"spine_04\"),Weight=0.0055),(Item=(Type=Bone,Name=\"neck_01\"),Weight=0.0315),(Item=(Type=Bone,Name=\"neck_02\"),Weight=0.0808),(Item=(Type=Bone,Name=\"clavicle_r\"),Weight=0.2361),(Item=(Type=Bo ...(truncated)"
    Weight=@n11.Value
   )
   > &RigUnit_ParentConstraint(
    AdvancedSettings=(RotationOrderForFilter=XZY)
    Child=(Type=Bone,Name="extra_collar_04_r")
    Parents="((Item=(Type=Bone,Name=\"deform_upperarm_fr_r\"),Weight=0.0067),(Item=(Type=Bone,Name=\"deform_upperarm_up_r\"),Weight=0.0093),(Item=(Type=Bone,Name=\"upperarm_twist_01_r\"),Weight=0.0227),(Item=(Type=Bone,Name=\"spine_03\"),Weight=0.0352),(Item=(Type=Bone,Name=\"d ...(truncated)"
    Weight=@n11.Value
   )
   head_vector_nodes = RigVMCollapseNode()
   RigVMVariableNode(Value=@head_vector_nodes.Result, ExecuteContext=@RigVMFunction_Sequence_4.B) {variable=head_vector}
   > RigVMVariableNode(Value=-90) {variable=rotate_collar}
   > RigVMFunction_Sequence_2 = RigVMAggregateNode()
   DotRot_CollarHead_1 = RigVMFunctionReferenceNode(
    BoneIn=(Type=Bone,Name="extra_collar_01_l")
    s_max=1
    s_min=0.85
    ExecuteContext=@RigVMFunction_Sequence_2.A
    angle=RigVMVariableNode() {variable=rotate_collar}.Value
   )
   > DotRot_CollarHead_2 = RigVMFunctionReferenceNode(BoneIn=(Type=Bone,Name="extra_collar_02_l"), s_max=1, s_min=0.85, angle=RigVMVariableNode() {variable=rotate_collar}.Value)
   > DotRot_CollarHead_3 = RigVMFunctionReferenceNode(BoneIn=(Type=Bone,Name="extra_collar_03_l"), s_max=1, s_min=0.75, angle=RigVMVariableNode() {variable=rotate_collar}.Value)
   > DotRot_CollarHead_5 = RigVMFunctionReferenceNode(BoneIn=(Type=Bone,Name="extra_collar_04_l"), s_max=1, s_min=0.85, angle=RigVMVariableNode() {variable=rotate_collar}.Value)
   DotRot_CollarHead = RigVMFunctionReferenceNode(
    BoneIn=(Type=Bone,Name="extra_collar_m")
    s_max=1
    s_min=0.9
    ExecuteContext=@RigVMFunction_Sequence_2.B
    angle=RigVMVariableNode() {variable=rotate_collar}.Value
   )
   DotRot_CollarHead_7 = RigVMFunctionReferenceNode(
    BoneIn=(Type=Bone,Name="extra_collar_01_r")
    s_max=1
    s_min=0.85
    ExecuteContext=@RigVMFunction_Sequence_2.C
    angle=RigVMVariableNode() {variable=rotate_collar}.Value
   )
   > DotRot_CollarHead_8 = RigVMFunctionReferenceNode(BoneIn=(Type=Bone,Name="extra_collar_02_r"), s_max=1, s_min=0.8, angle=RigVMVariableNode() {variable=rotate_collar}.Value)
   > DotRot_CollarHead_9 = RigVMFunctionReferenceNode(BoneIn=(Type=Bone,Name="extra_collar_03_r"), s_max=1, s_min=0.75, angle=RigVMVariableNode() {variable=rotate_collar}.Value)
   > DotRot_CollarHead_11 = RigVMFunctionReferenceNode(BoneIn=(Type=Bone,Name="extra_collar_04_r"), s_max=1, s_min=0.8, angle=RigVMVariableNode() {variable=rotate_collar}.Value)
   &RigUnit_ModifyTransforms(
    ExecutePin=()
    ItemToModify=((Item=(Type=Bone,Name="dyn_jacket_fr_r_base"),Transform=(Rotation=(X=0,Y=-0.2588,Z=0,W=0.9659),Translation=(X=-2,Y=0,Z=0),Scale3D=(X=1,Y=1,Z=1))))
    Weight=1
    WeightMinimum=0
   )
   &RigVMFunction_MathDoubleMul(B=1.5, A=@n11.Value)
   &RigVMFunction_MathDoubleMul(B=2, A=@n11.Value)
   &RigVMFunction_MathDoubleMul(B=1.5, A=@n11.Value)
   &RigVMFunction_MathDoubleMul(B=2, A=@n11.Value)
   &RigVMFunction_MathDoubleMul(A=@n5.OutputParam, B=RigVMVariableNode() {variable=targetAlpha}.Value)
   &RigVMFunction_MathDoubleMul(A=@n9.OutputParam, B=RigVMVariableNode() {variable=targetAlpha}.Value)
   &RigUnit_OffsetTransformForItem(
    ExecutePin=()
    Item=(Type=Bone,Name="None")
    OffsetTransform=(Rotation=(X=0,Y=0,Z=-0.7071,W=0.7071),Translation=(X=0,Y=-3,Z=0),Scale3D=(X=1,Y=1,Z=1))
    bPropagateToChildren=True
    Item.Name=@DotRot_CollarHead.BoneOut
    Weight=@DotRot_CollarHead.DotAlpha
   )
  ]}
]
hierarchy [
 root {loc=<0 0 0>, rot=<0 0 0 1>}
 pelvis {parent=root, loc=<0 -0.7071 93.1191>, rot=<0.0127 0.707 -0.0127 -0.707>}
 spine_01 {parent=pelvis, loc=<6.7494 -0.6388 0>, rot=<0 0 0.034 -0.9994>}
 spine_02 {parent=spine_01, loc=<7.2518 -0.035 0>, rot=<0 0 0.0238 -0.9997>}
 spine_03 {parent=spine_02, loc=<6.0104 0 0>, rot=<0 0 0.0149 -0.9999>}
 spine_04 {parent=spine_03, loc=<8.1024 0 0>, rot=<0 0 -0.0132 -0.9999>}
 spine_05 {parent=spine_04, loc=<10.0146 0 0>, rot=<0 0 -0.0015 -1>}
 neck_01 {parent=spine_05, loc=<11.3138 0.1529 0>, rot=<0 0 -0.2003 -0.9797>}
 neck_02 {parent=neck_01, loc=<4.5627 0 0>, rot=<0 0 0.0504 -0.9987>}
 head {parent=neck_02, loc=<5.3914 0 0>, rot=<0 0 0.115 -0.9934>}
 head_end {parent=head, loc=<10 0 0>, rot=<0 0 0 1>}
 attach_cape {parent=spine_05, loc=<9.103 -11.729 0.0016>, rot=<0.0617 -0.7045 -0.0616 -0.7044>, scale=<0.9996 1.0014 1.0003>}
 cape {parent=attach_cape, type=Null, loc=<0 -3.7234 -0.3522>, rot=<0.1382 0 0 0.9904>}
 attach_backpack {parent=spine_05, loc=<-0.0198 -13.4915 0.0028>, rot=<-0.0091 -0.7071 0.0093 -0.707>, scale=<0.9996 1.0014 1.0003>}
 Backpack_BR {parent=attach_backpack, type=Null, loc=<0 -1.1495 0.1228>, rot=<-0.0647 0 0 0.9979>}
 clavicle_l {parent=spine_05, loc=<4.3575 -0.1466 -2.3223>, rot=<0.0465 -0.7289 -0.009 -0.6829>}
 upperarm_l {parent=clavicle_l, loc=<15.5111 0 0>, rot=<0.0226 -0.3234 -0.0088 -0.9459>}
 lowerarm_l {parent=upperarm_l, loc=<26.6514 0 0>, rot=<0.0391 -0.0099 -0.2453 -0.9686>}
 hand_l {parent=lowerarm_l, loc=<22.3491 0 0>, rot=<-0.7081 -0.0507 0.0975 -0.6975>}
 thumb_01_l {parent=hand_l, loc=<1.3118 -0.3191 -2.459>, rot=<0.6546 -0.4042 0.0664 -0.6354>}
 thumb_02_l {parent=thumb_01_l, loc=<5.9336 -0.0448 0.0401>, rot=<-0.0037 0.0213 0.298 -0.9543>}
 thumb_03_l {parent=thumb_02_l, loc=<2.2223 0.3682 -0.0907>, rot=<0.0118 -0.013 0.0066 -0.9998>}
 index_metacarpal_l {parent=hand_l, loc=<1.6535 0.0183 -2.96>, rot=<0.0158 -0.0124 -0.0173 -0.9996>}
 index_01_l {parent=index_metacarpal_l, loc=<8.3008 0.0357 -0.0524>, rot=<-0.0044 0.0485 0.0924 -0.9945>}
 index_02_l {parent=index_01_l, loc=<3.3587 -0.3004 0.0452>, rot=<-0.002 0.0027 0.3225 -0.9466>}
 index_03_l {parent=index_02_l, loc=<2.5578 -0.123 -0.0252>, rot=<-0.002 0.0067 0.2558 -0.9667>}
 middle_metacarpal_l {parent=hand_l, loc=<2.1856 -0.1725 -0.5984>, rot=<0.006 0.0168 -0.0329 -0.9993>}
 middle_01_l {parent=middle_metacarpal_l, loc=<7.3275 -0.2447 0.0129>, rot=<-0.0208 0.0042 0.0858 -0.9961>}
 middle_02_l {parent=middle_01_l, loc=<4.0123 0.1401 0.1175>, rot=<-0.0218 0.0258 0.3512 -0.9357>}
 middle_03_l {parent=middle_02_l, loc=<2.6077 0.0534 0.1373>, rot=<0.0082 0.0188 0.1708 -0.9851>}
 ring_metacarpal_l {parent=hand_l, loc=<2.2074 -0.3625 1.3638>, rot=<-0.0284 0.0722 -0.0458 -0.9959>}
 ring_01_l {parent=ring_metacarpal_l, loc=<7.0295 -0.6197 -0.1446>, rot=<-0.0489 -0.0252 0.1892 -0.9804>}
 ring_02_l {parent=ring_01_l, loc=<3.6174 0.4647 -0.1027>, rot=<0 0 0.2806 -0.9598>}
 ring_03_l {parent=ring_02_l, loc=<2.4152 0.114 0.0199>, rot=<0 0 0.1189 -0.9929>}
 pinky_metacarpal_l {parent=hand_l, loc=<1.7616 -0.9753 3.3814>, rot=<-0.0397 0.0814 -0.069 -0.9935>}
 pinky_01_l {parent=pinky_metacarpal_l, loc=<7.1709 -0.4831 -0.1094>, rot=<-0.0654 -0.0035 0.2242 -0.9723>}
 pinky_02_l {parent=pinky_01_l, loc=<2.8274 0.0893 -0.1043>, rot=<0.0037 0.0011 0.2854 -0.9584>}
 pinky_03_l {parent=pinky_02_l, loc=<2.2046 0.0307 -0.0078>, rot=<-0.0002 0.0098 0.1445 -0.9895>}
 deform_wrist_dn_base_l {parent=hand_l, loc=<0 -2.2342 0>, rot=<0 0 0 1>}
 deform_wrist_dn_l {parent=deform_wrist_dn_base_l, loc=<0 0 0>, rot=<0 0 0 1>}
 deform_wrist_up_base_l {parent=hand_l, loc=<0 2.2682 0>, rot=<0 0 0 1>}
 deform_wrist_up_l {parent=deform_wrist_up_base_l, loc=<0 0 0>, rot=<0 0 0 1>}
 lowerarm_twist_02_l {parent=lowerarm_l, loc=<9.6199 0 0>, rot=<0 0 0 -1>}
 lowerarm_twist_01_l {parent=lowerarm_l, loc=<17.8549 0 0>, rot=<0 0 0 -1>}
 deform_elbow_in_base_l {parent=lowerarm_l, loc=<0.2112 5.259 0.8274>, rot=<0 0 0 -1>}
 deform_elbow_in_l {parent=deform_elbow_in_base_l, loc=<0 0 0>, rot=<0 0 0 1>}
 deform_elbow_out_base_l {parent=lowerarm_l, loc=<0 -4 0>, rot=<0 0 0 -1>}
 deform_elbow_out_l {parent=deform_elbow_out_base_l, loc=<0 0 0>, rot=<0 0 0 1>}
 upperarm_twist_01_l {parent=upperarm_l, loc=<0.0041 0 0>, rot=<0 0 0 -1>}
 deform_upperarm_bk_base_l {parent=upperarm_twist_01_l, loc=<1.0813 -7.2193 -1.9453>, rot=<0 0 0 1>}
 deform_upperarm_bk_l {parent=deform_upperarm_bk_base_l, loc=<0 0 0>, rot=<0 0 0 1>}
 deform_upperarm_fr_base_l {parent=upperarm_twist_01_l, loc=<-0.3691 7.0506 -2.765>, rot=<0 0 0 1>}
 deform_upperarm_fr_l {parent=deform_upperarm_fr_base_l, loc=<0 0 0>, rot=<0 0 0 1>}
 deform_upperarm_up_base_l {parent=upperarm_twist_01_l, loc=<-3.2216 -1.1391 3.3311>, rot=<0 0 0 1>}
 deform_upperarm_up_l {parent=deform_upperarm_up_base_l, loc=<0 0 0>, rot=<0 0 0 1>}
 upperarm_twist_02_l {parent=upperarm_l, loc=<13.7941 0 0>, rot=<0 0 0 -1>}
 clavicle_r {parent=spine_05, loc=<4.3575 -0.1466 2.3223>, rot=<0.7289 0.0465 -0.6829 0.009>}
 upperarm_r {parent=clavicle_r, loc=<-15.5111 0 0>, rot=<0.0226 -0.3234 -0.0088 -0.9459>}
 lowerarm_r {parent=upperarm_r, loc=<-26.6514 0 0>, rot=<0.0391 -0.0099 -0.2453 -0.9686>}
 hand_r {parent=lowerarm_r, loc=<-22.3491 0 0>, rot=<-0.7081 -0.0507 0.0975 -0.6975>}
 thumb_01_r {parent=hand_r, loc=<-1.3118 0.3191 2.459>, rot=<0.6546 -0.4042 0.0664 -0.6354>}
 thumb_02_r {parent=thumb_01_r, loc=<-5.9336 0.0448 -0.0401>, rot=<-0.0037 0.0213 0.298 -0.9543>}
 thumb_03_r {parent=thumb_02_r, loc=<-2.2223 -0.3682 0.0907>, rot=<0.0118 -0.013 0.0066 -0.9998>}
 index_metacarpal_r {parent=hand_r, loc=<-1.6535 -0.0183 2.96>, rot=<0.0158 -0.0124 -0.0173 -0.9996>}
 index_01_r {parent=index_metacarpal_r, loc=<-8.3008 -0.0357 0.0524>, rot=<-0.0044 0.0485 0.0924 -0.9945>}
 index_02_r {parent=index_01_r, loc=<-3.3587 0.3004 -0.0452>, rot=<-0.002 0.0027 0.3225 -0.9466>}
 index_03_r {parent=index_02_r, loc=<-2.5578 0.123 0.0252>, rot=<-0.002 0.0067 0.2558 -0.9667>}
 middle_metacarpal_r {parent=hand_r, loc=<-2.1856 0.1725 0.5984>, rot=<0.006 0.0168 -0.0329 -0.9993>}
 middle_01_r {parent=middle_metacarpal_r, loc=<-7.3275 0.2447 -0.0129>, rot=<-0.0208 0.0042 0.0858 -0.9961>}
 middle_02_r {parent=middle_01_r, loc=<-4.0123 -0.1401 -0.1175>, rot=<-0.0218 0.0258 0.3512 -0.9357>}
 middle_03_r {parent=middle_02_r, loc=<-2.6077 -0.0534 -0.1373>, rot=<0.0082 0.0188 0.1708 -0.9851>}
 ring_metacarpal_r {parent=hand_r, loc=<-2.2074 0.3625 -1.3638>, rot=<-0.0284 0.0722 -0.0458 -0.9959>}
 ring_01_r {parent=ring_metacarpal_r, loc=<-7.0295 0.6197 0.1446>, rot=<-0.0489 -0.0252 0.1892 -0.9804>}
 ring_02_r {parent=ring_01_r, loc=<-3.6174 -0.4647 0.1027>, rot=<0 0 0.2806 -0.9598>}
 ring_03_r {parent=ring_02_r, loc=<-2.4152 -0.114 -0.0199>, rot=<0 0 0.1189 -0.9929>}
 pinky_metacarpal_r {parent=hand_r, loc=<-1.7616 0.9753 -3.3814>, rot=<-0.0397 0.0814 -0.069 -0.9935>}
 pinky_01_r {parent=pinky_metacarpal_r, loc=<-7.1709 0.4831 0.1094>, rot=<-0.0654 -0.0035 0.2242 -0.9723>}
 pinky_02_r {parent=pinky_01_r, loc=<-2.8274 -0.0893 0.1043>, rot=<0.0037 0.0011 0.2854 -0.9584>}
 pinky_03_r {parent=pinky_02_r, loc=<-2.2046 -0.0307 0.0078>, rot=<-0.0002 0.0098 0.1445 -0.9895>}
 deform_wrist_dn_base_r {parent=hand_r, loc=<-0.0004 2.2345 -0.0001>, rot=<0 0 0 1>}
 deform_wrist_dn_r {parent=deform_wrist_dn_base_r, loc=<0 0 0>, rot=<0 0 0 1>}
 deform_wrist_up_base_r {parent=hand_r, loc=<0.0003 -2.2684 0.0001>, rot=<0 0 0 1>}
 deform_wrist_up_r {parent=deform_wrist_up_base_r, loc=<0 0 0>, rot=<0 0 0 1>}
 lowerarm_twist_02_r {parent=lowerarm_r, loc=<-9.6199 0 0>, rot=<0 0 0 -1>}
 lowerarm_twist_01_r {parent=lowerarm_r, loc=<-17.8549 0 0>, rot=<0 0 0 -1>}
 deform_elbow_in_base_r {parent=lowerarm_r, loc=<-0.2111 -5.2591 -0.8275>, rot=<0 0 0 -1>}
 deform_elbow_in_r {parent=deform_elbow_in_base_r, loc=<-0.0001 0 0.0002>, rot=<0 0 0 1>}
 deform_elbow_out_base_r {parent=lowerarm_r, loc=<0 4 -0.0001>, rot=<0 0 0 -1>}
 deform_elbow_out_r {parent=deform_elbow_out_base_r, loc=<0 0 0>, rot=<0 0 0 1>}
 upperarm_twist_01_r {parent=upperarm_r, loc=<-0.0041 0 0>, rot=<0 0 0 -1>}
 deform_upperarm_bk_base_r {parent=upperarm_twist_01_r, loc=<-1.0814 7.2192 1.9456>, rot=<0 0 0 1>}
 deform_upperarm_bk_r {parent=deform_upperarm_bk_base_r, loc=<0 0 0>, rot=<0 0 0 1>}
 deform_upperarm_fr_base_r {parent=upperarm_twist_01_r, loc=<0.3689 -7.0506 2.7652>, rot=<0 0 0 1>}
 deform_upperarm_fr_r {parent=deform_upperarm_fr_base_r, loc=<0 0 0>, rot=<0 0 0 1>}
 deform_upperarm_up_base_r {parent=upperarm_twist_01_r, loc=<3.2214 1.1391 -3.3308>, rot=<0 0 0 1>}
 deform_upperarm_up_r {parent=deform_upperarm_up_base_r, loc=<0 0 0>, rot=<0 0 0 -1>}
 upperarm_twist_02_r {parent=upperarm_r, loc=<-13.7941 0 0>, rot=<0 0 0 -1>}
 extra_collar_m {parent=spine_05, loc=<13.4411 -6.0024 0>, rot=<0 0 -0.0755 -0.9971>}
 extra_collar_01_r {parent=spine_05, loc=<13.2514 -4.8829 5.0295>, rot=<0.3773 0.0529 -0.0798 -0.9211>}
 extra_collar_02_r {parent=spine_05, loc=<12.5293 -1.258 7.6813>, rot=<0.5378 0.1113 -0.023 -0.8354>}
 extra_collar_03_r {parent=spine_05, loc=<11.5361 2.9691 8.7018>, rot=<0.6143 0.2445 0.0332 -0.7495>}
 extra_collar_04_r {parent=spine_05, loc=<9.5377 7.5765 8.4076>, rot=<0.7178 0.3366 -0.0004 -0.6095>}
 extra_collar_01_l {parent=spine_05, loc=<13.2509 -4.883 -5.0295>, rot=<-0.3773 -0.0529 -0.0798 -0.9211>}
 extra_collar_02_l {parent=spine_05, loc=<12.5291 -1.258 -7.6813>, rot=<-0.5378 -0.1113 -0.023 -0.8354>}
 extra_collar_03_l {parent=spine_05, loc=<11.5362 2.9691 -8.7018>, rot=<-0.6143 -0.2445 0.0332 -0.7495>}
 extra_collar_04_l {parent=spine_05, loc=<9.5375 7.5765 -8.4076>, rot=<-0.7178 -0.3366 -0.0004 -0.6095>}
 deform_pec_aim_base_l {parent=spine_04, loc=<7.8732 12.6388 -2.4123>, rot=<0.0102 -0.4602 0.3979 -0.7936>}
 deform_pec_aim_l {parent=deform_pec_aim_base_l, loc=<0 0 0>, rot=<0 0 0 1>}
 deform_pec_base_l {parent=deform_pec_aim_l, loc=<10.766 0 0>, rot=<0 0 0 1>}
 deform_pec_l {parent=deform_pec_base_l, loc=<0.055 0 0>, rot=<0 0 0 1>}
 deform_pec_aim_base_r {parent=spine_04, loc=<7.8727 12.6388 2.4123>, rot=<-0.4602 -0.0102 0.7936 0.3979>}
 deform_pec_aim_r {parent=deform_pec_aim_base_r, loc=<0 0 0>, rot=<0 0 0 1>}
 deform_pec_base_r {parent=deform_pec_aim_r, loc=<-10.7662 -0.0006 -0.0007>, rot=<0 0 0 1>}
 deform_pec_r {parent=deform_pec_base_r, loc=<0 0 0>, rot=<0 0 0 1>}
 dyn_jacket_fr_r_base {parent=spine_02, loc=<-1.4796 9.0545 10.6628>, rot=<0.0032 0.4719 0.8815 -0.016>}
 dyn_jacket_fr_r {parent=dyn_jacket_fr_r_base, loc=<0.1 0 0>, rot=<0 0 0 1>}
 dyn_jacket_side_r_base {parent=spine_02, loc=<-0.9931 -2.4091 16.3804>, rot=<0.03 -0.0181 0.9986 -0.0404>}
 dyn_jacket_side_r {parent=dyn_jacket_side_r_base, loc=<0.1 0 0>, rot=<0 0 0 1>}
 dyn_jacket_bk_r_base {parent=spine_02, loc=<-0.5033 -12.6462 9.8351>, rot=<0.0167 -0.5861 0.81 -0.0083>}
 dyn_jacket_bk_r {parent=dyn_jacket_bk_r_base, loc=<0.1 0 0>, rot=<0 0 0 1>}
 dyn_jacket_fr_l_base {parent=spine_02, loc=<-1.4791 9.0545 -10.6628>, rot=<-0.4719 0.0032 -0.016 -0.8815>}
 dyn_jacket_fr_l {parent=dyn_jacket_fr_l_base, loc=<0.1 0 0>, rot=<0 -1 0 0>}
 dyn_jacket_side_l_base {parent=spine_02, loc=<-0.9931 -2.4091 -16.3804>, rot=<0.0181 0.03 -0.0404 -0.9986>}
 dyn_jacket_side_l {parent=dyn_jacket_side_l_base, loc=<0.1 0 0>, rot=<0 -1 0 0>}
 dyn_jacket_bk_l_base {parent=spine_02, loc=<-0.5038 -12.6462 -9.8351>, rot=<0.5861 0.0167 -0.0083 -0.81>}
 dyn_jacket_bk_l {parent=dyn_jacket_bk_l_base, loc=<0.1 0 0>, rot=<0 -1 0 0>}
 thigh_l {parent=pelvis, loc=<-2.5693 -0.0109 -7.7501>, rot=<0.0012 0.0572 0.0155 -0.9982>}
 calf_l {parent=thigh_l, loc=<-39.0484 0 0>, rot=<0.0006 -0.0161 -0.0391 -0.9991>}
 foot_l {parent=calf_l, loc=<-42.561 0.1077 0.4292>, rot=<0.0418 -0.0455 0.0343 -0.9975>}
 ball_l {parent=foot_l, loc=<-6.1599 13.8264 0.6806>, rot=<0.0001 -0.0001 -0.7063 -0.7079>}
 calf_twist_02_l {parent=calf_l, loc=<-18.8064 -0.0967 0.0397>, rot=<0.0237 0.0095 -0.0035 -0.9997>}
 calf_twist_01_l {parent=calf_l, loc=<-35.813 -0.0967 0.0397>, rot=<0.0237 0.0095 -0.0035 -0.9997>}
 deform_knee_in_base_l {parent=calf_l, loc=<1.298 -5.2957 -0.0496>, rot=<0 0 0 -1>}
 deform_knee_in_l {parent=deform_knee_in_base_l, loc=<0 0 0>, rot=<0 0 0 1>}
 deform_knee_out_base_l {parent=calf_l, loc=<0 4 0>, rot=<0 0 0 -1>}
 deform_knee_out_l {parent=deform_knee_out_base_l, loc=<0 0 0>, rot=<0 0 0 1>}
 thigh_twist_01_l {parent=thigh_l, loc=<-21.6303 0 0>, rot=<0.0013 0 0 -1>}
 deform_groin_base_l {parent=thigh_l, loc=<-2.1908 8.1142 -1.7442>, rot=<0 0 0 -1>}
 deform_groin_l {parent=deform_groin_base_l, loc=<0 0 0>, rot=<0 0 0 1>}
 thigh_r {parent=pelvis, loc=<-2.5693 -0.0109 7.7501>, rot=<0.0572 -0.0012 0.9982 0.0155>}
 calf_r {parent=thigh_r, loc=<39.0484 0 0>, rot=<0.0006 -0.0161 -0.0391 -0.9991>}
 foot_r {parent=calf_r, loc=<42.561 -0.1077 -0.4292>, rot=<0.0418 -0.0455 0.0343 -0.9975>}
 ball_r {parent=foot_r, loc=<6.1599 -13.8264 -0.6806>, rot=<0.0001 -0.0001 -0.7063 -0.7079>}
 calf_twist_02_r {parent=calf_r, loc=<18.8064 0.0967 -0.0397>, rot=<0.0237 0.0095 -0.0035 -0.9997>}
 calf_twist_01_r {parent=calf_r, loc=<35.813 0.0967 -0.0397>, rot=<0.0237 0.0095 -0.0035 -0.9997>}
 deform_knee_in_base_r {parent=calf_r, loc=<-1.298 5.2957 0.0496>, rot=<0 0 0 -1>}
 deform_knee_in_r {parent=deform_knee_in_base_r, loc=<0 0 0>, rot=<0 0 0 1>}
 deform_knee_out_base_r {parent=calf_r, loc=<0 -4 0>, rot=<0 0 0 -1>}
 deform_knee_out_r {parent=deform_knee_out_base_r, loc=<0 0 0>, rot=<0 0 0 1>}
 thigh_twist_01_r {parent=thigh_r, loc=<21.6303 0 0>, rot=<0.0013 0 0 -1>}
 deform_groin_base_r {parent=thigh_r, loc=<2.1908 -8.1142 1.7442>, rot=<0 0 0 -1>}
 deform_groin_r {parent=deform_groin_base_r, loc=<0 0 0>, rot=<0 0 0 1>}
 deform_glute_base_l {parent=pelvis, loc=<-8.8064 -8.8222 -8.4287>, rot=<0.0275 0.0501 -0.4472 -0.8926>}
 deform_glute_l {parent=deform_glute_base_l, loc=<0 0 0>, rot=<0 0 0 1>}
 deform_glute_base_r {parent=pelvis, loc=<-8.8062 -8.8222 8.4287>, rot=<0.0501 -0.0275 0.8926 -0.4472>}
 deform_glute_r {parent=deform_glute_base_r, loc=<0 0 0>, rot=<0 0 0 1>}
 extra_glute_r {parent=pelvis, loc=<-8.8657 -8.9023 8.4352>, rot=<0.0501 -0.0275 0.8926 -0.4472>}
 extra_glute_l {parent=pelvis, loc=<-8.8659 -8.9023 -8.4352>, rot=<0.0275 0.0501 -0.4472 -0.8926>}
]
variables [
 NewVar {type=bool, default=False}
 contrs {type=double, default=0}
 crouch {type=bool, default=False}
 head_vector {type=FVector, default=(X=0,Y=0,Z=0)}
 rotate_collar {type=double, default=0}
 target {type=bool, default=False}
 targetAlpha {type=double, default=0}
]
