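# Physics asset for the M_MED_CreamSkull_Dove male body cosmetic, set up for secondary motion of six jacket flaps: front, side and back on each side.
# Each flap hangs from a small base capsule locked in rotation to spine_01, with limited swing and tight twist, looser swing at the front.
# Pelvis, spine_01 and thigh capsules form a minimal lower-body proxy joined at the pelvis.
# The front-left flap alone carries two capsules, the second offset and rotated.
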PhysicsAsset /BRCosmetics/Characters/Player/Male/Medium/Bodies/M_MED_CreamSkull_Dove/Meshes/M_MED_CreamSkull_Dove_Physics
bodies {
 dyn_jacket_bk_l [{shapeType=Capsule, length=15, radius=3, shapeName=dyn_jacket_bk_l_capsule, center=<15 0 0>, rotation=<rot 0 0 90>}]
 dyn_jacket_bk_l_base [{shapeType=Capsule, length=2, radius=1, shapeName=dyn_jacket_bk_l_base_capsule}]
 dyn_jacket_bk_r [{shapeType=Capsule, length=15, radius=3, shapeName=dyn_jacket_bk_r_capsule, center=<15 0 0>, rotation=<rot 0 0 90>}]
 dyn_jacket_bk_r_base [{shapeType=Capsule, length=2, radius=1, shapeName=dyn_jacket_bk_r_base_capsule}]
 dyn_jacket_fr_l [
  {shapeType=Capsule, length=15, radius=3, shapeName=dyn_jacket_fr_l_capsule, center=<17 0 0>, rotation=<rot 0 0 90>}
  {shapeType=Capsule, length=15, radius=3, shapeName=dyn_jacket_fr_l_capsule, center=<17 -2.9091 -1.6796>, rotation=<rot 0 0 60>}
 ]
 dyn_jacket_fr_l_base [{shapeType=Capsule, length=2, radius=1, shapeName=dyn_jacket_fr_l_base_capsule}]
 dyn_jacket_fr_r [{shapeType=Capsule, length=15, radius=4, shapeName=dyn_jacket_fr_r_capsule, center=<15 0 0>, rotation=<rot 0 0 90>}]
 dyn_jacket_fr_r_base [{shapeType=Capsule, length=2, radius=1, shapeName=dyn_jacket_fr_r_base_capsule}]
 dyn_jacket_side_l [{shapeType=Capsule, length=15, radius=3, shapeName=dyn_jacket_side_l_capsule, center=<15 0 0>, rotation=<rot 0 0 90>}]
 dyn_jacket_side_l_base [{shapeType=Capsule, length=2, radius=1, shapeName=dyn_jacket_side_l_base_capsule}]
 dyn_jacket_side_r [{shapeType=Capsule, length=15, radius=3, shapeName=dyn_jacket_side_r_capsule, center=<15 0 0>, rotation=<rot 0 0 90>}]
 dyn_jacket_side_r_base [{shapeType=Capsule, length=2, radius=1, shapeName=dyn_jacket_side_r_base_capsule}]
 pelvis [{shapeType=Capsule, length=10, radius=10, center=<-2 -0.5 0>}]
 spine_01 [{shapeType=Capsule, length=10, radius=10}]
 thigh_l [{shapeType=Capsule, length=20, radius=7.2, center=<-20 0.005 0>, rotation=<rot -90 -0.0173 0>}]
 thigh_r [{shapeType=Capsule, length=20, radius=7.2, center=<20 -0.005 0>, rotation=<rot -90 179.9827 0>}]
}
constraints [
 {bone1Name=spine_01, bone2Name=pelvis, swing1LimitDegrees=45, swing2LimitDegrees=45, twistLimitDegrees=45}
 {bone1Name=thigh_l, bone2Name=pelvis, swing1LimitDegrees=45, swing2LimitDegrees=45, twistLimitDegrees=45}
 {bone1Name=thigh_r, bone2Name=pelvis, swing1LimitDegrees=45, swing2LimitDegrees=45, twistLimitDegrees=45}
 {
  bone1Name=dyn_jacket_bk_l_base
  bone2Name=spine_01
  swing1LimitDegrees=45
  swing1Motion=Locked
  swing2LimitDegrees=45
  swing2Motion=Locked
  twistLimitDegrees=45
  twistMotion=Locked
 }
 {bone1Name=dyn_jacket_bk_l, bone2Name=dyn_jacket_bk_l_base, swing1LimitDegrees=10, swing2LimitDegrees=20, twistLimitDegrees=5}
 {
  bone1Name=dyn_jacket_bk_r_base
  bone2Name=spine_01
  swing1LimitDegrees=45
  swing1Motion=Locked
  swing2LimitDegrees=45
  swing2Motion=Locked
  twistLimitDegrees=45
  twistMotion=Locked
 }
 {bone1Name=dyn_jacket_bk_r, bone2Name=dyn_jacket_bk_r_base, swing1LimitDegrees=10, swing2LimitDegrees=20, twistLimitDegrees=5}
 {
  bone1Name=dyn_jacket_fr_l_base
  bone2Name=spine_01
  swing1LimitDegrees=45
  swing1Motion=Locked
  swing2LimitDegrees=45
  swing2Motion=Locked
  twistLimitDegrees=45
  twistMotion=Locked
 }
 {bone1Name=dyn_jacket_fr_l, bone2Name=dyn_jacket_fr_l_base, swing1LimitDegrees=30, swing2LimitDegrees=45, twistLimitDegrees=5}
 {
  bone1Name=dyn_jacket_fr_r_base
  bone2Name=spine_01
  swing1LimitDegrees=45
  swing1Motion=Locked
  swing2LimitDegrees=45
  swing2Motion=Locked
  twistLimitDegrees=45
  twistMotion=Locked
 }
 {bone1Name=dyn_jacket_fr_r, bone2Name=dyn_jacket_fr_r_base, swing1LimitDegrees=30, swing2LimitDegrees=45, twistLimitDegrees=10}
 {
  bone1Name=dyn_jacket_side_l_base
  bone2Name=spine_01
  swing1LimitDegrees=45
  swing1Motion=Locked
  swing2LimitDegrees=45
  swing2Motion=Locked
  twistLimitDegrees=45
  twistMotion=Locked
 }
 {bone1Name=dyn_jacket_side_l, bone2Name=dyn_jacket_side_l_base, swing1LimitDegrees=10, swing2LimitDegrees=20, twistLimitDegrees=5}
 {
  bone1Name=dyn_jacket_side_r_base
  bone2Name=spine_01
  swing1LimitDegrees=45
  swing1Motion=Locked
  swing2LimitDegrees=45
  swing2Motion=Locked
  twistLimitDegrees=45
  twistMotion=Locked
 }
 {bone1Name=dyn_jacket_side_r, bone2Name=dyn_jacket_side_r_base, swing1LimitDegrees=10, swing2LimitDegrees=20, twistLimitDegrees=5}
]
