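# Animation Blueprint for the M_MED_CreamSkull_Dove male body cosmetic, adding secondary motion to the pose copied from its parent mesh.
# A pre-dynamics Control Rig, fed crouch and targeting states, leads into a RigidBody simulation in spine_02 space, cached as startPose.
# A front-end idle curve blends DefaultSlot over that pose from extra_sleeve_cuff down.
# The event graph updates those inputs and computes a gravity override through a deprecated helper.
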AnimBlueprint /BRCosmetics/Characters/Player/Male/Medium/Bodies/M_MED_CreamSkull_Dove/Meshes/M_MED_CreamSkull_Dove_AnimBP
target_skeleton=&/M_MED_CreamSkull_Dove_Skeleton
graphs [
 AnimGraph {nodes [
   Output Pose(Result=Layered blend per bone(
     BasePose=Use cached pose 'startPose'()
     BlendPoses_0=Slot 'DefaultSlot'()
     BlendWeights_0=Get customIdleIsPlayingFloat() -> (customIdleIsPlayingFloat:float)
    ) {layerSetup [{branchFilters [{boneName=extra_sleeve_cuff}]}]}) {layerGroup=None}
   n1 = Control Rig!(
    Source=Copy Pose From Mesh() {bUseAttachedParent=true}
    Alpha
    crouch:bool=Get Crouching() -> (Crouching:bool)
    target:bool=Get Targeting() -> (Targeting:bool)
   ) -> (Pose) {controlRigAssetReference {blueprintRigClass=&/M_MED_CreamSkull_Dove_CtrlRig_PreDyn.M_MED_CreamSkull_Dove_CtrlRig_PreDyn_C}}
   n2 = RigidBody!(
    ComponentPose=Local To Component(LocalPose=@n1)
    bAlphaBoolEnabled:bool
    OverrideWorldGravity:Vector="0,0,0"
    ComponentLinearAccScale:Vector="0,0,0"
    ComponentLinearVelScale:Vector="0,0,0"
    bTransferBoneVelocities:bool=False
   ) -> (Pose) {
    alphaInputType=Bool
    bClampLinearTranslationLimitToRefPose=true
    bFreezeIncomingPoseOnStart=true
    bOverrideWorldGravity=true
    bTransferBoneVelocities=true
    overridePhysicsAsset=&/M_MED_CreamSkull_Dove_Physics
    simulationSpace=BaseBoneSpace
    alphaBoolBlend {blendOption=Linear}
    baseBoneRef {boneName=spine_02}
    componentAppliedLinearAccClamp=<100 100 100>
    componentLinearAccScale=<0.6 0.2 0.2>
    componentLinearVelScale=<0.6 0.2 0.2>
    simSpaceSettings {
     dampingAlpha=0
     maxAngularAcceleration=50
     maxAngularVelocity=8
     maxLinearAcceleration=2000
     maxLinearVelocity=500
     worldAlpha=1
     externalLinearDragV=<1 1 1>
    }
   }
   Save cached pose 'startPose'(Pose=Component To Local(ComponentPose=@n2)) {cacheName=startPose}
  ]}
 EventGraph {nodes [
   Event BlueprintUpdateAnimation()
   > n1 = Sequence!(execute) -> (then_0, then_1, then_2:Exec, then_3:Exec)
   Set Targeting(execute=@n1.then_1, Targeting:bool=Get bIsTargeting()) -> (Output_Get:bool)
   Set Crouching(execute=@n1.then_2, Crouching:bool=Get bIsCrouching()) -> (Output_Get:bool)
   Branch(execute=@n1.then_3, Condition=Get bIsInFrontEnd())
   > Set customIdleIsPlayingFloat(customIdleIsPlayingFloat:float=GetCurveValue(self=GetOwnerFrontEndAnimInstance(), CurveName=Frontend_Idle_Bypass)) -> (Output_Get:float)
   n2 = Gravity Override State Function DEPRECATED(
    execute=@n1.then_0
    CharacterPartAnimInstance=Self-Reference()
    GameTimeInSeconds=GetGameTimeInSeconds()
    GravityOverrideParams=Get Gravity Override Params()
    DebugJointName=spine_05
   )
   > Set GravityOverride(GravityOverride=@n2)
  ]}
]
variables [
 Crouching {type=boolean}
 customIdleIsPlayingFloat {type=number}
 Gravity Override Params {type=GravityOverrideParamsStruct}
 GravityOverride {type=Vector}
 Targeting {type=boolean}
]
