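# Physics asset for the face accessory mesh of the male medium CreamSkull Dove character body, covering three dynamic hair pieces: center, left and right.
# A large head capsule anchors three small base capsules, each joined to it with swing and twist locked.
# Each base connects to a hair capsule through a constraint without those locks, forming short two-link chains that let the hair move relative to the head.
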PhysicsAsset /BRCosmetics/Characters/Player/Male/Medium/Bodies/M_MED_CreamSkull_Dove/Meshes/Parts/M_MED_CreamSkull_Dove_FaceAcc_Physics
bodies {
 dyn_hair_center [{shapeType=Capsule, length=3, radius=2, shapeName=dyn_hair_center_capsule, center=<5 0 0>, rotation=<rot 0 0 90>}]
 dyn_hair_center_base [{shapeType=Capsule, length=0.505, radius=0.505, shapeName=dyn_hair_center_base_capsule, rotation=<rot 0 -90 0>}]
 dyn_hair_l [{shapeType=Capsule, length=3, radius=2, shapeName=dyn_hair_l_capsule, center=<5 0 0>, rotation=<rot 0 0 90>}]
 dyn_hair_l_base [{shapeType=Capsule, length=0.505, radius=0.505, shapeName=dyn_hair_l_base_capsule, rotation=<rot 0 -90 0>}]
 dyn_hair_r [{shapeType=Capsule, length=3, radius=2, shapeName=dyn_hair_r_capsule, center=<5 0 0>, rotation=<rot 0 0 90>}]
 dyn_hair_r_base [{shapeType=Capsule, length=0.505, radius=0.505, shapeName=dyn_hair_r_base_capsule, rotation=<rot 0 -90 0>}]
 head [{shapeType=Capsule, length=13.4334, radius=9.1754, rotation=<rot 90 0 0>}]
}
constraints [
 {
  bone1Name=dyn_hair_center_base
  bone2Name=head
  swing1LimitDegrees=45
  swing1Motion=Locked
  swing2LimitDegrees=45
  swing2Motion=Locked
  twistLimitDegrees=45
  twistMotion=Locked
 }
 {bone1Name=dyn_hair_center, bone2Name=dyn_hair_center_base, swing1LimitDegrees=30, swing2LimitDegrees=60, twistLimitDegrees=5}
 {
  bone1Name=dyn_hair_l_base
  bone2Name=head
  swing1LimitDegrees=45
  swing1Motion=Locked
  swing2LimitDegrees=45
  swing2Motion=Locked
  twistLimitDegrees=45
  twistMotion=Locked
 }
 {bone1Name=dyn_hair_l, bone2Name=dyn_hair_l_base, swing1LimitDegrees=30, swing2LimitDegrees=60, twistLimitDegrees=5}
 {
  bone1Name=dyn_hair_r_base
  bone2Name=head
  swing1LimitDegrees=45
  swing1Motion=Locked
  swing2LimitDegrees=45
  swing2Motion=Locked
  twistLimitDegrees=45
  twistMotion=Locked
 }
 {bone1Name=dyn_hair_r, bone2Name=dyn_hair_r_base, swing1LimitDegrees=30, swing2LimitDegrees=60, twistLimitDegrees=5}
]
